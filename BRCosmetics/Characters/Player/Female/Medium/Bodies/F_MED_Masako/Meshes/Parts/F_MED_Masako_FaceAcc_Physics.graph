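# Physics asset for the face accessory mesh of the female medium Masako character, covering secondary motion of hair and ponytail bones.
# Box bodies cover three dynamic hair strand bones, capsules cover two ponytail bones, and a small capsule sits on the head.
# Each hair and ponytail body is constrained directly to the head rather than chained, with limited swing and tight twist, slightly wider swing for strands than ponytails.
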PhysicsAsset /BRCosmetics/Characters/Player/Female/Medium/Bodies/F_MED_Masako/Meshes/Parts/F_MED_Masako_FaceAcc_Physics
bodies {
 dyn_hair_01_l [{shapeType=Box, extentX=5, extentY=1, extentZ=7.793, center=<3 0 0>, rotation=<rot 90 0 0>}]
 dyn_hair_01_r [{shapeType=Box, extentX=2.5813, extentY=1, extentZ=2.3394, center=<0.9769 0 0>, rotation=<rot 90 0 0>}]
 dyn_hair_02_r [{shapeType=Box, extentX=2.5609, extentY=1, extentZ=5.2598, center=<2.2956 0 0>, rotation=<rot 90 0 0>}]
 head [{shapeType=Capsule, length=1, radius=1}]
 ponytail_A [{shapeType=Capsule, length=2.1646, radius=3.5994, center=<3.0785 0.0203 3.1583>, rotation=<rot 90 0 0>}]
 ponytail_C [{shapeType=Capsule, length=2.1646, radius=3.5994, center=<3.9022 -0.1103 -2.2108>, rotation=<rot 90 0 0>}]
}
constraints [
 {bone1Name=ponytail_A, bone2Name=head, swing1LimitDegrees=40, swing2LimitDegrees=40, twistLimitDegrees=10}
 {bone1Name=ponytail_C, bone2Name=head, swing1LimitDegrees=40, swing2LimitDegrees=40, twistLimitDegrees=10}
 {bone1Name=dyn_hair_01_l, bone2Name=head, swing1LimitDegrees=45, swing2LimitDegrees=45, twistLimitDegrees=5}
 {bone1Name=dyn_hair_01_r, bone2Name=head, swing1LimitDegrees=45, swing2LimitDegrees=45, twistLimitDegrees=5}
 {bone1Name=dyn_hair_02_r, bone2Name=head, swing1LimitDegrees=45, swing2LimitDegrees=45, twistLimitDegrees=5}
]
